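; Burst gameplay cue for the strike of the HighTower Tapas hammer throw, adjusting its Cascade particle effects.
; After the parent OnBurst runs, it continues only if the cue's instigator is valid and casts the source object to an actor.
; The second particle component moves to that actor, facing back toward the instigator.
; The first component keeps its position and rotates along the direction from instigator to that actor.
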(fn OnBurst(Cascade)
  (params (MyTarget :type Actor%20Object%20Reference) (Parameters :type Gameplay%20Cue%20Parameters%20Structure%20%28by%20ref%29) (ParticleComponents :type Array%20of%20Cascade%20Particle%20System%20Component%20Object%20References) (AudioComponents :type Array%20of%20Audio%20Component%20Object%20References) (BurstCameraShakeInstance :type Camera%20Shake%20Base%20Object%20Reference) (BurstDecalInstance :type Decal%20Actor%20Object%20Reference))
  (bind _returnvalue (Transformation|GetActorLocation (.Instigator (Ability|GameplayCue|BreakGameplayCueParameters Parameters))))
  (bind (_normalizedmagnitude _rawmagnitude _effectcontext _matchedtagname _originaltag _aggregatedsourcetags _aggregatedtargettags _location _normal _instigator _effectcauser _sourceobject) (Ability|GameplayCue|BreakGameplayCueParameters Parameters))
  (|CallParentFunction MyTarget Parameters ParticleComponents AudioComponents BurstCameraShakeInstance BurstDecalInstance)
  ;[1 1 1] This is the hit Target
  (Utilities|IsValid _instigator
    (:Is%20Valid
      (bind _asactor (Utilities|Casting|CastToActor _sourceobject))
      (Transformation|SetWorldLocationAndRotation (Utilities|Array|Get(acopy) ParticleComponents 1) (Transformation|GetActorLocation _asactor) (Math|Rotator|FindLookatRotation (Transformation|GetActorLocation _asactor) _returnvalue) false false)
      ;[1 1 1] This is the Projectile
      (Transformation|SetWorldRotation (Utilities|Array|Get(acopy) ParticleComponents) (Math|Rotator|FindLookatRotation _returnvalue (Transformation|GetActorLocation _asactor)) false false)
      (return))))
